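# Character-parent material for the coconut shell emote prop, adding time-of-day glow and reactive lighting between the shared PreFX and PostFX functions.
# An emissive texture with optional Fresnel forms the base glow, while time-of-day scalars vary glow strength and base color brightness.
# A packed mask supplies dot, gradient and edge light for the reactive emissive, driven by ReactiveCurve.
# Two local-space sphere masks blended by PositionCurve provide impact regions, with a debug view switch.
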Material /BRCosmetics/Animation/Props/Gadget_CoconutShell/Materials/M_Emote_CoconutShell
graph {
 properties {
  bUseMaterialAttributes=true
  bUsedWithMeshDeformer=true
  bUsedWithMorphTargets=true
  bUsedWithNanite=true
  bUsedWithNiagaraMeshParticles=true
  bUsedWithSkeletalMesh=true
  pixelDepthOffsetMode=PDOM_Legacy
  shadingModel=MSM_FromMaterialExpression
  usedShadingModels=Default Lit | Subsurface | Clear Coat | Subsurface Profile | Cloth
 }
 nodes [
  n1 = MaterialFunctionCall() {materialFunction=&MF_CharacterParent_2023_PreFX}
  n2 = GetMaterialAttributes(None=@n1.Result) {attributeGetTypes [#1, #2]}
  n3 = TextureSampleParameter2D() {group=TechArt, parameterName=CoconutEM_RE, texture=/BRCosmetics/Animation/Props/Gadget_CoconutShell/Textures/T_CoconutShell_EmoteProp_E}
  n4 = Multiply(A=@n3.RGB, B=NamedRerouteUsage())
  n5 = Multiply(A=Fresnel() {baseReflectFraction=0, exponent=1}, B=@n4)
  n6 = NamedRerouteUsage()
  n7 = ScalarParameter() {group=NightEM, parameterName=ReactiveCurve}
  n8 = Add(
   A=@n2.Emissive Color
   B=Add(A=StaticSwitchParameter(True=@n5, False=@n4) {defaultValue=true, group=TechArt, parameterName=UseFresnel_EM}, B=Multiply(A=@n6, B=@n7))
  )
  n9 = StaticSwitchParameter(
   True=LinearInterpolate(A=NamedRerouteUsage(), B=NamedRerouteUsage(), Alpha=ScalarParameter() {group=Reactive, parameterName=Mask02Debug})
   False=Constant()
  ) {group=Reactive, parameterName=DebugSphereMask}
  n10 = NamedRerouteUsage()
  MaterialFunctionCall(
   Material=SetMaterialAttributes!(MaterialAttributes=@n2.MaterialAttributes, Emissive Color=Add(A=@n8, B=@n9), Base Color=Multiply(A=@n2.Base Color, B=@n10)) {attributeSetTypes [#1, #2]}
   M=@n1.M
   Pre Skinned Local Position=@n1.M
   Distance Blend=@n1.M
   Fresnel=@n1.M
  ) {materialFunction=&MF_CharacterParent_2023_PostFX}
  n11 = ScalarParameter() {defaultValue=0.8, group=NightEM, parameterName=Day_Intensity}
  NamedRerouteDeclaration(None=MaterialFunctionCall(
    Value_Morning=@n11
    Value_Day=@n11
    Value_Evening=ScalarParameter() {defaultValue=1, group=NightEM, parameterName=Evening_Intensity}
    Value_Night=ScalarParameter() {defaultValue=2, group=NightEM, parameterName=Night_Intensity}
   ) {materialFunction=&Scalar_Time_of_Day}.Result) {name=NightAlpha01, nodeColor=<rgba 0 0.0391 1 1>}
  n12 = ScalarParameter() {defaultValue=1, group=NightEM, parameterName=Day_Intensity_BaseColor}
  NamedRerouteDeclaration(None=MaterialFunctionCall(
    Value_Morning=@n12
    Value_Day=@n12
    Value_Evening=ScalarParameter() {defaultValue=0.7, group=NightEM, parameterName=Evening_Intensity_BaseColor}
    Value_Night=ScalarParameter() {defaultValue=0.4, group=NightEM, parameterName=Night_Intensity_BaseColor}
   ) {materialFunction=&Scalar_Time_of_Day}.Result) {name=NightAlpha02, nodeColor=<rgba 0 0.0391 1 1>}
  NamedRerouteDeclaration(None=@n3.RGB) {name=CocoEM, nodeColor=<rgba 1 0 0.2812 1>}
  n13 = NamedRerouteUsage()
  n14 = ScalarParameter() {defaultValue=0.5, group=NightEM, parameterName=DefaultIntensity}
  n15 = TextureSampleParameter2D() {
   group=NightEM
   parameterName=ReactiveLight
   samplerType=SAMPLERTYPE_Masks
   texture=/BRCosmetics/Animation/Props/Gadget_CoconutShell/Textures/T_coco_EmoteProp_FX
  }
  n16 = Multiply(A=ScalarParameter() {defaultValue=4, group=NightEM, parameterName=DotIntensity}, B=Multiply(A=@n13, B=@n15.R))
  n17 = Multiply(A=ScalarParameter() {defaultValue=1.5, group=NightEM, parameterName=GradientIntensity}, B=Multiply(A=@n13, B=@n15.G))
  n18 = NamedRerouteDeclaration(None=@n15.B) {name=EdgeLight, nodeColor=<rgba 0 1 0.8359 1>}
  n19 = Add(
   A=Add(A=Multiply(A=@n13, B=@n14), B=Add(A=@n16, B=@n17))
   B=Multiply(A=ScalarParameter() {defaultValue=4, group=NightEM, parameterName=EdgetIntensity}, B=Multiply(A=@n13, B=@n18))
  )
  n20 = NamedRerouteUsage()
  NamedRerouteDeclaration(None=Multiply(A=NamedRerouteUsage(), B=Multiply(A=@n19, B=@n20))) {name=ReactiveEM, nodeColor=<rgba 0 1 0.7188 1>}
  n21 = ScalarParameter() {defaultValue=100, group=Reactive, parameterName=ImpactRadius}
  n22 = ScalarParameter() {defaultValue=0.5, group=Reactive, parameterName=ImpactSoftness}
  n23 = SphereMask(
   A=VectorParameter() {group=Reactive, parameterName=SphereCenter01, defaultValue=<rgba 0 0 0 1>}.RGB
   B=LocalPosition().XYZ
   Radius=@n21
   Hardness=@n22
  )
  n24 = SphereMask(
   A=VectorParameter() {group=Reactive, parameterName=SphereCenter02, defaultValue=<rgba 0 0 0 1>}.RGB
   B=LocalPosition().XYZ
   Radius=@n21
   Hardness=@n22
  )
  NamedRerouteDeclaration(None=LinearInterpolate(A=@n23, B=@n24, Alpha=ScalarParameter() {group=NightEM, parameterName=PositionCurve})) {name=Reactive01, nodeColor=<rgba 0 0.9766 1 1>}
  n25 = ScalarParameter() {defaultValue=1, group=NightEM, parameterName=Day_Intensity_React}
  NamedRerouteDeclaration(None=MaterialFunctionCall(
    Value_Morning=@n25
    Value_Day=@n25
    Value_Evening=ScalarParameter() {defaultValue=1.5, group=NightEM, parameterName=Evening_Intensity_React}
    Value_Night=ScalarParameter() {defaultValue=2.5, group=NightEM, parameterName=Night_Intensity_React}
   ) {materialFunction=&Scalar_Time_of_Day}.Result) {name=NightAlpha03, nodeColor=<rgba 0 0.0391 1 1>}
  NamedRerouteDeclaration(None=@n24) {name=Mask02, nodeColor=<rgba 0 1 0.0625 1>}
  NamedRerouteDeclaration(None=@n23) {name=Mask01, nodeColor=<rgba 0.5 1 0 1>}
  NamedRerouteUsage()
 ]
}
