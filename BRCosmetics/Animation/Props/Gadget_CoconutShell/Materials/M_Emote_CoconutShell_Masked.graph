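# Masked, two-sided character-style material for the coconut shell emote prop, wrapped between the CharacterParent 2023 PreFX and PostFX functions.
# It overrides base color, emissive and opacity mask, with time-of-day scalars, tint parameters, a mask texture and optional Fresnel emissive.
# A switchable world position offset pushes vertices along normals in a sine wave, limited by texture and UV masks.
# A Niagara dynamic parameter drives the wave phase, with a time-based debug blend.
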Material /BRCosmetics/Animation/Props/Gadget_CoconutShell/Materials/M_Emote_CoconutShell_Masked
graph {
 properties {
  bUseMaterialAttributes=true
  bUsedWithMeshDeformer=true
  bUsedWithMorphTargets=true
  bUsedWithNanite=true
  bUsedWithNiagaraMeshParticles=true
  bUsedWithSkeletalMesh=true
  blendMode=BLEND_Masked
  pixelDepthOffsetMode=PDOM_Legacy
  shadingModel=MSM_FromMaterialExpression
  twoSided=true
  usedShadingModels=Default Lit | Subsurface | Clear Coat | Subsurface Profile | Cloth
 }
 nodes [
  n1 = MaterialFunctionCall() {materialFunction=&MF_CharacterParent_2023_PreFX}
  n2 = GetMaterialAttributes(None=@n1.Result) {attributeGetTypes [#1, #2]}
  n3 = Multiply(A=@n2.Base Color, B=VectorParameter() {group=NightEM, parameterName=Basecolor_Adjustment, defaultValue=<rgba 1 1 1 1>}.RGB)
  n4 = NamedRerouteUsage()
  n5 = SetMaterialAttributes!(
   MaterialAttributes=@n2.MaterialAttributes
   Emissive Color=Add(A=@n2.Emissive Color, B=NamedRerouteUsage())
   Opacity Mask=ChannelMaskParameter(None=@n1.Tech Art Mask) {parameterName=OpacityChannel_RE}.R
   Base Color=Multiply(A=LinearInterpolate(A=@n2.Base Color, B=@n3, Alpha=@n4), B=NamedRerouteUsage())
   World Position Offset=NamedRerouteUsage()
  ) {attributeSetTypes [
    #1
    #3
    #2
    #4
   ]}
  MaterialFunctionCall(Material=@n5, M=@n1.M, Pre Skinned Local Position=@n1.M, Distance Blend=@n1.M, Fresnel=@n1.M) {materialFunction=&MF_CharacterParent_2023_PostFX}
  n6 = ScalarParameter() {defaultValue=0.5, group=NightEM, parameterName=Day_Intensity}
  NamedRerouteDeclaration(None=MaterialFunctionCall(
    Value_Morning=@n6
    Value_Day=@n6
    Value_Evening=ScalarParameter() {defaultValue=1, group=NightEM, parameterName=Evening_Intensity}
    Value_Night=ScalarParameter() {defaultValue=2, group=NightEM, parameterName=Night_Intensity}
   ) {materialFunction=&Scalar_Time_of_Day}.Result) {name=NightAlpha01, nodeColor=<rgba 0 0.0391 1 1>}
  n7 = ScalarParameter() {defaultValue=1, group=NightEM, parameterName=Day_Intensity_BaseColor}
  NamedRerouteDeclaration(None=MaterialFunctionCall(
    Value_Morning=@n7
    Value_Day=@n7
    Value_Evening=ScalarParameter() {defaultValue=0.8, group=NightEM, parameterName=Evening_Intensity_BaseColor}
    Value_Night=ScalarParameter() {defaultValue=0.4, group=NightEM, parameterName=Night_Intensity_BaseColor}
   ) {materialFunction=&Scalar_Time_of_Day}.Result) {name=NightAlpha02, nodeColor=<rgba 0 0.0391 1 1>}
  n8 = TextureSampleParameter2D() {
   group=NightEM
   parameterName=AdjustmentMask
   samplerType=SAMPLERTYPE_Masks
   texture=/BRCosmetics/Animation/Props/Gadget_CoconutShell/Textures/T_coco_EmoteProp_TwoSided_FX
  }
  NamedRerouteDeclaration(None=@n8.R) {name=SpriteArea, nodeColor=<rgba 0.5938 0 1 1>}
  n9 = TextureSampleParameter2D() {group=TechArt, parameterName=CoconutEM_RE, texture=/BRCosmetics/Animation/Props/Gadget_CoconutShell/Textures/T_CoconutShell_EmoteProp_E}
  n10 = Multiply(A=@n9.RGB, B=NamedRerouteUsage())
  n11 = StaticSwitchParameter(True=Multiply(A=Fresnel() {baseReflectFraction=0, exponent=1}, B=@n10), False=@n10) {group=TechArt, parameterName=UseFresnel_EM}
  n12 = VectorParameter() {group=NightEM, parameterName=EMcolor_Adjustment, defaultValue=<rgba 1 1 1 1>}
  NamedRerouteDeclaration(None=LinearInterpolate(A=@n11, B=Multiply(A=@n11, B=@n12.RGB), Alpha=NamedRerouteUsage())) {name=EM_OUT, nodeColor=<rgba 0 1 0.2734 1>}
  n13 = VertexNormalWS()
  n14 = Time()
  n15 = ScalarParameter() {defaultValue=-1, parameterName=WPO Speed, sliderMax=5, sliderMin=-5}
  n16 = ComponentMask(None=TextureCoordinate()) {g=true}
  n17 = Add(
   A=LinearInterpolate(
    A=DynamicParameter() {defaultValue=<rgba -1 1 1 1>, paramNames [WPO Offset, Param2, Param3, Param4]}.WPO Offset
    B=Multiply(A=@n14, B=@n15)
    Alpha=ScalarParameter() {parameterName=WPO Debug}
   )
   B=Power(Base=Multiply(A=@n16) {constB=5}, Exp=ScalarParameter() {defaultValue=2, parameterName=WPO Stretch, sliderMax=5})
  )
  n18 = Multiply(
   A=ConstantBiasScale(None=Sine(None=@n17)) {bias=0.2, scale=1}
   B=ScalarParameter() {defaultValue=10, parameterName=WPO Intensity, sliderMax=50}
  ) {constB=100}
  n19 = TextureCoordinate()
  n20 = Multiply(
   A=Multiply(
    A=Multiply(A=@n13, B=@n18)
    B=TextureSample() {samplerType=SAMPLERTYPE_Masks, texture=&T_coco_EmoteProp_TwoSided_FX}.R
   )
   B=If(A=ComponentMask(None=@n19) {r=true}, "A > B"=Constant() {r=1}, "A < B"=Constant()) {constB=0.5}
  )
  n21 = SmoothStep(
   Min=ScalarParameter() {parameterName=WPO Min, sliderMax=1}
   Max=ScalarParameter() {defaultValue=1, parameterName=WPO Max, sliderMax=1}
   Value=ComponentMask(None=TextureCoordinate()) {g=true}
  )
  NamedRerouteDeclaration(None=StaticSwitchParameter(True=Multiply(A=@n20, B=@n21), False=Constant()) {parameterName=WPO}) {name=WPO, nodeColor=<rgba 0.0781 1 0 1>}
 ]
}
